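# Sound cue for the Unstable Bow's glitch-in effect, played through the weapons sound class with glitch attenuation.
# The first two quality levels layer a random pick of four glitch-in waves, two slightly delayed bump waves and a pitched-up hum, each with modulation.
# The other two levels play only a random pick of two glitch-in waves.
# A concurrency override caps simultaneous instances, stopping the oldest.
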SoundCue /PrimalGameplay/Sounds/Weapons/Bows/Unstable/Bow_Unstable_Glitch_In_Cue
root=QualityLevel_0
nodes [
 {id=Delay_0, children [Mixer_1], properties {delayMax=0.05, delayMin=0.05}}
 {id=Mixer_0, children [Modulator_0,Modulator_1,Modulator_2], properties {inputVolume [1,0.6,0.5]}}
 {id=Mixer_1, children [WavePlayer_21, WavePlayer_22], properties {inputVolume [1,0.3]}}
 {id=Modulator_0, children [Random_2]}
 {id=Modulator_1, children [Delay_0], properties {pitchMax=1, pitchMin=1}}
 {id=Modulator_2, children [WavePlayer_0], properties {pitchMax=1.15, pitchMin=1.15}}
 {id=QualityLevel_0, children [Mixer_0,Mixer_0,Random_3,Random_3]}
 {id=Random_2, children [WavePlayer_8, WavePlayer_9, WavePlayer_10, WavePlayer_11], properties {weights [1,1,1,1]}}
 {id=Random_3, children [WavePlayer_12, WavePlayer_13], properties {weights [1,1]}}
 {id=WavePlayer_0, properties {soundWaveAssetPtr=&/Bow_Unstable_Glitch_InHum_01}}
 {id=WavePlayer_10, properties {soundWaveAssetPtr=&/Bow_Unstable_Glitch_In_03}}
 {id=WavePlayer_11, properties {soundWaveAssetPtr=&/Bow_Unstable_Glitch_In_04}}
 {id=WavePlayer_12, properties {soundWaveAssetPtr=&/Bow_Unstable_Glitch_In_01}}
 {id=WavePlayer_13, properties {soundWaveAssetPtr=&/Bow_Unstable_Glitch_In_02}}
 {id=WavePlayer_21, properties {soundWaveAssetPtr=&/Bow_Unstable_Glitch_Bump_02}}
 {id=WavePlayer_22, properties {soundWaveAssetPtr=&/Bow_Unstable_Glitch_Bump_03}}
 {id=WavePlayer_8, properties {soundWaveAssetPtr=&Bow_Unstable_Glitch_In_01}}
 {id=WavePlayer_9, properties {soundWaveAssetPtr=&Bow_Unstable_Glitch_In_02}}
]
properties {
 attenuationSettings=/Game/Sounds/Attenuation/Fort_Comsumable_Glitch_UnstableBow_Att
 bOverrideConcurrency=true
 maxDistance=3500
 pitchMultiplier=0.9
 soundClassObject=/Game/Sounds/FortSoundClassesAndModes/Fort_Weapons
 volumeMultiplier=0.5926
 concurrencyOverrides {maxCount=4, resolutionRule=StopOldest}
}
