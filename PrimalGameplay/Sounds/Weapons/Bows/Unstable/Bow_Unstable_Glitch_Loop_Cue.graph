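# Sound cue for the looping glitch sound of the Unstable Bow weapon.
# A single wave player feeds the glitch loop wave straight to the output, at reduced volume in the gadgets sound class.
# It uses a dedicated Unstable Bow glitch attenuation asset and overrides concurrency to cap simultaneous instances, stopping the oldest first.
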SoundCue /PrimalGameplay/Sounds/Weapons/Bows/Unstable/Bow_Unstable_Glitch_Loop_Cue
root=WavePlayer_0
nodes [{id=WavePlayer_0, properties {soundWaveAssetPtr=&/Bow_Unstable_Glitch_Loop_01}}]
properties {
 attenuationSettings=/Game/Sounds/Attenuation/Fort_Comsumable_Glitch_UnstableBow_Att
 bOverrideConcurrency=true
 maxDistance=3500
 soundClassObject=&Fort_Gadgets
 volumeMultiplier=0.4
 concurrencyOverrides {maxCount=4, resolutionRule=StopOldest}
}
