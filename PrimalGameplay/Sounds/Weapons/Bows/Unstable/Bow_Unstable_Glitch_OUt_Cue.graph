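# Sound cue for the Unstable Bow's glitch-out effect, routed through the weapons sound class with its own glitch attenuation.
# A mixer layers a randomly chosen glitch-out variant, picked evenly from three waves, over a quieter hum layer at a fixed raised pitch.
# The whole cue plays lowered in pitch and volume, with overridden concurrency that caps simultaneous instances and stops the oldest.
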SoundCue /PrimalGameplay/Sounds/Weapons/Bows/Unstable/Bow_Unstable_Glitch_OUt_Cue.Bow_Unstable_Glitch_Out_Cue
root=Mixer_0
nodes [
 {id=Mixer_0, children [Modulator_0,Modulator_1], properties {inputVolume [1,0.5]}}
 {id=Modulator_0, children [Random_2]}
 {id=Modulator_1, children [WavePlayer_0], properties {pitchMax=1.3, pitchMin=1.3}}
 {id=Random_2, children [WavePlayer_7,WavePlayer_8,WavePlayer_9], properties {weights [1,1,1]}}
 {id=WavePlayer_0, properties {soundWaveAssetPtr=&/Bow_Unstable_Glitch_OutHum_01}}
 {id=WavePlayer_7, properties {soundWaveAssetPtr=&/Bow_Unstable_Glitch_Out_01}}
 {id=WavePlayer_8, properties {soundWaveAssetPtr=&/Bow_Unstable_Glitch_Out_02}}
 {id=WavePlayer_9, properties {soundWaveAssetPtr=&/Bow_Unstable_Glitch_Out_03}}
]
properties {
 attenuationSettings=/Game/Sounds/Attenuation/Fort_Comsumable_Glitch_UnstableBow_Att
 bOverrideConcurrency=true
 maxDistance=3500
 pitchMultiplier=0.8
 soundClassObject=/Game/Sounds/FortSoundClassesAndModes/Fort_Weapons
 volumeMultiplier=0.5926
 concurrencyOverrides {maxCount=4, resolutionRule=StopOldest}
}
